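# ColorCalibration is a widget blueprint subclass of the native FortColorCalibrationScreen, serving as the color calibration screen in the FortniteCoreUI settings.
# Its tree is just an Overlay root holding InputBlocker, a fully transparent Image variable that fills the overlay, presumably to catch input behind the screen.
# The event graph is empty and no variables are declared, so any calibration behavior comes from the parent class.
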WidgetBlueprint /FortniteCoreUI/Settings/ColorCalibration
graphs [EventGraph {links [], nodes []}]
variables []
widgets {
 info {inheritedWidgetCount=0, parentClass=/Script/FortniteUI.FortColorCalibrationScreen, rootWidgetClass=&Overlay, widgetCount=2}
 widgets [
  Overlay_1 {widgetClassPath=&Overlay, widget=Overlay_1}
  InputBlocker {
   widgetClassPath=&Image
   bIsVariable=true
   parent=Overlay_1
   widget=InputBlocker
   properties {colorAndOpacity=<rgba 0 0 0 0>}
   slot {class=OverlaySlot, properties {horizontalAlignment=HAlign_Fill, verticalAlignment=VAlign_Fill}}
  }
 ]
}
