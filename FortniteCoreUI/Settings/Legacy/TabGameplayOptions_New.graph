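# Legacy gameplay options tab widget, based on FortTabGameplayOptions_TempNew, that hosts a single inherited SettingsPanel as its root.
# Named actions from the panel asynchronously push the ColorCalibration screen, then configure it for either color calibration or a full-range test.
# A SettingValueChanged custom event broadcasts the tab's setting-changed delegate, and CenterOnTab centers on the settings panel.
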WidgetBlueprint /FortniteCoreUI/Settings/Legacy/TabGameplayOptions_New
graphs [EventGraph {nodes [
   n1 = "On Execute Named Action (Settings_Panel)"!() -> (OutputDelegate, then, Setting:Fort Setting Object Reference, Action:Name=None) {is=title K2Node_ComponentBoundEvent}
   > n2 = Switch on Name!(execute:Exec, Selection:Name=@n1.Action) -> (Default:Exec, ViewColorCalibration:Exec, ViewFullRangeTest:Exec) {is=title K2Node_SwitchName, pinNames [ViewColorCalibration, ViewFullRangeTest]}
   n3 = PushPrimaryContentClassAsync!(
    execute:Exec=@n2.ViewColorCalibration
    WidgetClass:Common Activatable Widget Class Reference=/FortniteCoreUI/Settings/ColorCalibration.ColorCalibration_C
   ) -> (then:Exec, OnPrimaryContentPushed:Exec, ActivatableWidget:Common Activatable Widget Object Reference) {is=title K2Node_AsyncAction}
   n4 = Cast To ColorCalibration!(execute=@n3.OnPrimaryContentPushed, Object=@n3) -> (then, CastFailed, AsColor Calibration:Color Calibration Object Reference) {is=title K2Node_DynamicCast}
   > SetupForColorConfiguration(self:Fort Color Calibration Screen Object Reference=@n4)
   n5 = PushPrimaryContentClassAsync(execute=@n2.ViewFullRangeTest, WidgetClass=&ColorCalibration_C)
   n6 = Cast To ColorCalibration(execute=@n5.OnPrimaryContentPushed, Object=@n5)
   > SetupForFullRangeTest(self:Fort Color Calibration Screen Object Reference=@n6)
   SettingValueChanged!() -> (OutputDelegate, then, Value:double=~) {is=title K2Node_CustomEvent}
   > Call On Tab Setting Changed!(execute:Exec, self:Fort Options Tab Object Reference) -> (then:Exec) {is=title K2Node_CallDelegate}
   Event CenterOnTab() {is=title K2Node_Event}
   > CenterWidget!(execute, ToCenter:Widget Object Reference=Get Settings_Panel() -> (Settings_Panel:Fort Settings Panel Object Reference)) -> (then)
  ]}]
variables []
widgets {
 info {
  inheritedWidgetCount=1
  parentClass=/Script/FortniteUI.FortTabGameplayOptions_TempNew
  rootWidgetClass=/FortniteCoreUI/Settings/SettingsPanel.SettingsPanel_C
  widgetCount=1
 }
 widgets [Settings_Panel {widgetClassPath=/FortniteCoreUI/Settings/SettingsPanel, bInherited=true, bIsVariable=true, widget=Settings_Panel}]
}
