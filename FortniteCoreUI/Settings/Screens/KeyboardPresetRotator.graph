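# Settings widget, based on FortKeyboardPresetRotator, for cycling keyboard/mouse input presets with chevron buttons and showing each preset's description.
# On construct, it fills TextRotator with preset names for the current SubGame and selects index 0.
# Rotation sets SelectedIndex and refreshes Description and SubDescription, though those lookups pass a fixed Campaign SubGame.
# Helper functions replace options, set the selection, and toggle description visibility.
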WidgetBlueprint /FortniteCoreUI/Settings/Screens/KeyboardPresetRotator
graphs [
 EventGraph {nodes [
   "On Button Clicked (ButtonRight)"!() -> (OutputDelegate, then, Button:Common Button Legacy Object Reference) {is=title K2Node_ComponentBoundEvent}
   n1 = Get TextRotator() -> (TextRotator:Text Rotator White Object Reference)
   > ShiftTextRight(self:Common Rotator Object Reference=@n1)
   "On Button Clicked (ButtonLeft)"!() -> (OutputDelegate, then, Button:Common Button Legacy Object Reference) {is=title K2Node_ComponentBoundEvent}
   > ShiftTextLeft(self:Common Rotator Object Reference=@n1)
   n2 = "On Rotated (TextRotator)"!() -> (OutputDelegate, then, Value:int=0) {is=title K2Node_ComponentBoundEvent}
   > n3 = Set SelectedIndex(SelectedIndex:int=@n2.Value) -> (Output_Get:int=0)
   n4 = GetKBMPlayerInputPresetDescriptionForSubGame!(
    self:Fort Global UIContext Object Reference=Get FortGlobalUIContext!() -> (ReturnValue:Fort Global UIContext Object Reference) {is=title K2Node_GetBlueprintContextEx}
    SubGame:ESubGame Enum=Campaign
    CurrentPreset:int=@n3
   ) -> (InputPresetDescription:Text, InputPresetSubDescription:Text)
   > "SetText (Text)"(self=Get Description() -> (Description:Common Text Object Reference), InText=@n4)
   > "SetText (Text)"(self=Get SubDescription() -> (SubDescription:Common Text Object Reference), InText=@n4.InputPresetSubDescription)
   Event Construct()
   > Initialize()
  ]}
 Initialize {nodes [
   Initialize() {is=title K2Node_FunctionEntry}
   n1 = Get FortGlobalUIContext()
   > PopulateTextLabels!(
    execute
    self:Common Rotator Object Reference=Get TextRotator()
    Labels:Array of Texts=GetKBMPlayerInputPresetNamesForSubGame!(
     self:Fort Global UIContext Object Reference=@n1
     SubGame:ESubGame Enum=GetSubGame!(self:Fort Global UIContext Object Reference=@n1) -> (ReturnValue:ESubGame Enum=Campaign)
    ) -> (InputPresetNames:Array of Strings, InputPresetFriendlyNames:Array of Texts).InputPresetFriendlyNames
   ) -> (then)
   > Update Row Selector!(execute, self, Currently Selected:int=0) -> (then)
   n2 = GetKBMPlayerInputPresetDescriptionForSubGame(self=@n1, SubGame=Campaign, CurrentPreset=0)
   > "SetText (Text)"(self=Get Description(), InText=@n2)
   > "SetText (Text)"(self=Get SubDescription(), InText=@n2.InputPresetSubDescription)
  ]}
 Show Param Name {nodes [
   n1 = Show Param Name!() -> (then, Show:bool=~) {is=title K2Node_FunctionEntry}
   > SetVisibility(self=[Get Description(), Get SubDescription()], InVisibility=Select(Option 0=Collapsed, Option 1=SelfHitTestInvisible, Index=@n1))
  ]}
 Update Options {nodes [
   n1 = Update Options!() -> (then, NewParam:Array of Texts) {is=title K2Node_FunctionEntry}
   > PopulateTextLabels(self=Get TextRotator(), Labels=@n1)
  ]}
 Update Row Selector {nodes [
   n1 = Update Row Selector!() -> (then, Currently Selected:int) {is=title K2Node_FunctionEntry}
   > n2 = Set SelectedIndex(SelectedIndex=@n1) -> (Output_Get=0)
   > SetSelectedItem!(execute, self:Common Rotator Object Reference=Get TextRotator(), InValue:int=@n2) -> (then)
  ]}
]
variables [Spacer Width {type=number}]
widgets {
 info {inheritedWidgetCount=0, parentClass=/Script/FortniteUI.FortKeyboardPresetRotator, rootWidgetClass=&VerticalBox, widgetCount=10}
 widgets [
  VerticalBox_0 {widgetClassPath=&VerticalBox, widget=VerticalBox_0}
  SizeBox_0 {
   widgetClassPath=&SizeBox
   parent=VerticalBox_0
   widget=SizeBox_0
   properties {bOverride_MinDesiredWidth=true, minDesiredWidth=800}
   slot {class=VerticalBoxSlot, properties {horizontalAlignment=HAlign_Center, padding=<ltrb 32 0 32 0>}}
  }
  HorizontalBox_0 {
   widgetClassPath=&HorizontalBox
   parent=SizeBox_0
   widget=HorizontalBox_0
   slot {class=SizeBoxSlot, properties {horizontalAlignment=HAlign_Center}}
  }
  ButtonLeft {
   widgetClassPath=/FortUILibrary/Foundation/Buttons/PanelButton
   bIsVariable=true
   parent=HorizontalBox_0
   widget=ButtonLeft
   properties {
    bIsFocusable=false
    clickMethod=MouseDown
    inputActionUseRimBrush=false
    inputActionVerticalAlignment=VAlign_Center
    minHeight=64
    minWidth=32
    onButtonBaseClicked=(&self:WidgetTree.ButtonLeft.HandleOnButtonClicked)
    onButtonBaseDoubleClicked=(&self:WidgetTree.ButtonLeft.HandleOnButtonDoubleClicked)
    onButtonBaseHovered=(&self:WidgetTree.ButtonLeft.HandleOnButtonHovered)
    onButtonBaseUnhovered=(&self:WidgetTree.ButtonLeft.HandleOnButtonUnhovered)
    onSelectedChangedBase=(&self:WidgetTree.ButtonLeft.HandleOnSelectedChanged)
    renderOpacity=0.7
    style=/FortUILibrary/Foundation/Buttons/ButtonStyles/ButtonStyle_PageChevron_Left.ButtonStyle_PageChevron_Left_C
    hoveredSlateSoundOverride {resourceObject=/Game/Sounds/Fort_UI_Sounds/FrontEnd/New_FrontEnd/Fortnite_UI_MainMenu_Tab_Hover_01_Cue}
    inputActionRenderTranslation=<24 0>
    pressedSlateSoundOverride {resourceObject=/Game/Sounds/Fort_UI_Sounds/FrontEnd/New_FrontEnd/Fortnite_UI_MainMenu_Generic_01_Cue}
    triggeringInputAction {dataTable=&InputActionDataTable, rowName=ItemShop_Carousel_TabLeft}
   }
   slot {class=HorizontalBoxSlot, properties {horizontalAlignment=HAlign_Left, verticalAlignment=VAlign_Center, padding=<ltrb 0 0 32 0>}}
  }
  HorizontalBox_1 {
   widgetClassPath=&HorizontalBox
   parent=HorizontalBox_0
   widget=HorizontalBox_1
   slot {class=HorizontalBoxSlot, properties {verticalAlignment=VAlign_Center}}
  }
  Spacer_0 {
   widgetClassPath=&Spacer
   bIsVariable=true
   parent=HorizontalBox_1
   widget=Spacer_0
   properties {size=<0 56>}
   slot {class=HorizontalBoxSlot}
  }
  TextRotator {
   widgetClassPath=/FortUILibrary/Foundation/Buttons/TextRotatorWhite
   bIsVariable=true
   parent=HorizontalBox_1
   widget=TextRotator
   properties {minWidth=640, style=null}
   slot {class=HorizontalBoxSlot}
  }
  ButtonRight {
   widgetClassPath=&PanelButton
   bIsVariable=true
   parent=HorizontalBox_0
   widget=ButtonRight
   properties {
    bIsFocusable=false
    clickMethod=MouseDown
    inputActionUseRimBrush=false
    inputActionVerticalAlignment=VAlign_Center
    minHeight=64
    minWidth=32
    onButtonBaseClicked=(&self:WidgetTree.ButtonRight.HandleOnButtonClicked)
    onButtonBaseDoubleClicked=(&self:WidgetTree.ButtonRight.HandleOnButtonDoubleClicked)
    onButtonBaseHovered=(&self:WidgetTree.ButtonRight.HandleOnButtonHovered)
    onButtonBaseUnhovered=(&self:WidgetTree.ButtonRight.HandleOnButtonUnhovered)
    onSelectedChangedBase=(&self:WidgetTree.ButtonRight.HandleOnSelectedChanged)
    renderOpacity=0.7
    style=/FortUILibrary/Foundation/Buttons/ButtonStyles/ButtonStyle_PageChevron_Right.ButtonStyle_PageChevron_Right_C
    hoveredSlateSoundOverride {resourceObject=/Game/Sounds/Fort_UI_Sounds/FrontEnd/DefaultSounds/PressedButtons/MSS_UI_Button_Hover_Default}
    inputActionRenderTranslation=<-24 0>
    pressedSlateSoundOverride {resourceObject=/Game/Sounds/Fort_UI_Sounds/FrontEnd/DefaultSounds/PressedButtons/MSS_UI_Button_Pressed_Default}
    triggeringInputAction {dataTable=&InputActionDataTable, rowName=ItemShop_Carousel_TabRight}
   }
   slot {class=HorizontalBoxSlot, properties {horizontalAlignment=HAlign_Right, verticalAlignment=VAlign_Center, padding=<ltrb 32 0 0 0>}}
  }
  Description {
   widgetClassPath=&CommonTextBlock
   bIsVariable=true
   parent=VerticalBox_0
   widget=Description
   properties {
    lineHeightPercentage=1.29
    scrollStyle=&TextScrollStyle-Base_C
    style=/FortUILibrary/Styles/Typography/_TextStyle_Body1._TextStyle_Body1_C
    text=Description
    font {
     bForceMonospaced=false
     bMaterialIsStencil=false
     fontMaterial=null
     fontObject=&BurbankSmall
     letterSpacing=0
     monospacedWidth=1
     size=18
     skewAmount=0
     typefaceFontName=Bold
     outlineSettings {
      bApplyOutlineToDropShadows=false
      bMiteredCorners=false
      bSeparateFillAlpha=false
      outlineMaterial=null
      outlineSize=0
      outlineColor=<rgba 0 0 0 1>
     }
    }
    shadowOffset=<0 0>
    strikeBrush {
     bIsDynamicallyLoaded=false
     drawAs=NoDrawType
     imageType=NoImage
     mirroring=NoMirror
     resourceName=None
     resourceObject=null
     tiling=NoTile
     imageSize=<32 32>
     margin=<ltrb 0 0 0 0>
     outlineSettings {
      bUseBrushTransparency=false
      roundingType=HalfHeightRadius
      width=0
      color {colorUseRule=UseColor_Specified, specifiedColor=<rgba 0 0 0 0>}
      cornerRadii=<0 0 0 0>
     }
     tintColor {colorUseRule=UseColor_Specified, specifiedColor=<rgba 1 1 1 1>}
     uVRegion {bIsValid=false, max=<0 0>, min=<0 0>}
    }
   }
   slot {class=VerticalBoxSlot, properties {horizontalAlignment=HAlign_Center, padding=<ltrb 0 32 0 0>}}
  }
  SubDescription {
   widgetClassPath=&CommonTextBlock
   bIsVariable=true
   parent=VerticalBox_0
   widget=SubDescription
   properties {
    lineHeightPercentage=1.29
    scrollStyle=&TextScrollStyle-Base_C
    style=&_TextStyle_Body1_C
    text=Sub Description
    font {
     bForceMonospaced=false
     bMaterialIsStencil=false
     fontMaterial=null
     fontObject=&BurbankSmall
     letterSpacing=0
     monospacedWidth=1
     size=18
     skewAmount=0
     typefaceFontName=Bold
     outlineSettings {
      bApplyOutlineToDropShadows=false
      bMiteredCorners=false
      bSeparateFillAlpha=false
      outlineMaterial=null
      outlineSize=0
      outlineColor=<rgba 0 0 0 1>
     }
    }
    shadowOffset=<0 0>
    strikeBrush {
     bIsDynamicallyLoaded=false
     drawAs=NoDrawType
     imageType=NoImage
     mirroring=NoMirror
     resourceName=None
     resourceObject=null
     tiling=NoTile
     imageSize=<32 32>
     margin=<ltrb 0 0 0 0>
     outlineSettings {
      bUseBrushTransparency=false
      roundingType=HalfHeightRadius
      width=0
      color {colorUseRule=UseColor_Specified, specifiedColor=<rgba 0 0 0 0>}
      cornerRadii=<0 0 0 0>
     }
     tintColor {colorUseRule=UseColor_Specified, specifiedColor=<rgba 1 1 1 1>}
     uVRegion {bIsValid=false, max=<0 0>, min=<0 0>}
    }
   }
   slot {class=VerticalBoxSlot, properties {horizontalAlignment=HAlign_Center, padding=<ltrb 0 16 0 8>}}
  }
 ]
}
