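# Legacy options-menu tab for mouse and keyboard bindings, built around an inherited settings panel.
# The SelectKBMBindingPreset action asynchronously creates a preset rotator and shows it inside a "Choose Preset" confirmation dialog.
# The options menu's back action is disabled while the dialog is open, then re-enabled on result; confirming resets bindings to the selected preset.
# It also broadcasts setting changes and centers the panel on CenterOnTab.
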WidgetBlueprint /FortniteCoreUI/Settings/Legacy/TabMouseAndKeyboardBindingOptions_New
graphs [
 EventGraph {nodes [
   n1 = "On Execute Named Action (Settings_Panel)"!() -> (OutputDelegate, then, Setting:Fort Setting Object Reference, Action:Name=None) {is=title K2Node_ComponentBoundEvent}
   > n2 = Switch on Name!(execute:Exec, Selection:Name=@n1.Action) -> (Default:Exec, SelectKBMBindingPreset:Exec) {is=title K2Node_SwitchName, pinNames [SelectKBMBindingPreset]}
   n3 = CreateWidgetAsync!(
    execute:Exec=@n2.SelectKBMBindingPreset
    UserWidgetSoftClass:User Widget Soft Class Reference=/FortniteCoreUI/Settings/Screens/KeyboardPresetRotator.KeyboardPresetRotator_C
    OwningPlayer:Player Controller Object Reference=GetOwningPlayer()
    bSuspendInputUntilComplete:bool=true
   ) -> (then:Exec, OnComplete:Exec, OnFailure:Exec, UserWidget:User Widget Object Reference) {is=title K2Node_AsyncAction}
   n4 = ShowSimpleConfirmationDialog_CustomInput!(
    execute:Exec=@n3.OnComplete
    Icon:Slate Brush Structure
    Title:Text="LOCTEXT(\"Choose Preset\")"
    Message:Text
    ConfirmAction:Name=SettingPopupAccept
    DeclineAction:Name=SettingPopupDecline
    AdditionalContent:Widget Object Reference=@n3
    LeftAdditionalContent:Widget Object Reference
    Cancelable:bool=false
    TargetPlayer:Local Player Object Reference
   ) -> (then:Exec, DialogResult:Exec, Result:EFortDialogResult Enum, ResultName:Name) {is=title K2Node_AsyncAction}
   > SetOptionsMenuBackActionStateFromHandle!(execute, self, State:EInputActionState Enum=Disabled) -> (then)
   SetOptionsMenuBackActionStateFromHandle(execute=@n4.DialogResult, State=Enabled)
   > Branch(Condition="Equal (Enum)"(A:EFortDialogResult Enum=@n4, B:EFortDialogResult Enum=Confirmed))
   > ResetKeyboardMousePlayerBindingsToPreset!(
    execute
    LocalPlayerController:Fort Player Controller Object Reference=Cast To FortPlayerController!(Object=GetOwningPlayer()) -> (AsFort Player Controller, bSuccess:bool=~)
    PresetToSet:int=Get SelectedIndex!(self:Fort Keyboard Preset Rotator Object Reference=Cast To KeyboardPresetRotator!(Object=@n3) -> (AsKeyboard Preset Rotator:Keyboard Preset Rotator Object Reference, bSuccess:bool=~) {is=title K2Node_DynamicCast}) -> (SelectedIndex:int=0)
   ) -> (then)
   SettingValueChanged!() -> (OutputDelegate, then, Value:double=~) {is=title K2Node_CustomEvent}
   > Call On Tab Setting Changed!(execute:Exec, self:Fort Options Tab Object Reference) -> (then:Exec) {is=title K2Node_CallDelegate}
   Event CenterOnTab() {is=title K2Node_Event}
   > CenterWidget!(execute, ToCenter:Widget Object Reference=Get Settings_Panel() -> (Settings_Panel:Fort Settings Panel Object Reference)) -> (then)
  ]}
 SetOptionsMenuBackActionStateFromHandle {nodes [
   n1 = SetOptionsMenuBackActionStateFromHandle!() -> (then, State:EInputActionState Enum=Enabled) {is=title K2Node_FunctionEntry}
   n2 = Get OptionsMenu() -> (OptionsMenu:Options Menu Object Reference)
   > Branch(Condition=IsValid(Object=@n2))
   > SetBackActionStateFromHandle!(execute, self:Fort Options Menu Object Reference=@n2, State:EInputActionState Enum=@n1) -> (then)
  ]}
]
variables [OptionsMenu {type=OptionsMenu_C, default=null}]
widgets {
 info {
  inheritedWidgetCount=1
  parentClass=/Script/FortniteUI.FortTabMouseAndKeyboardBindingsOptions_TempNew
  rootWidgetClass=/FortniteCoreUI/Settings/SettingsPanel.SettingsPanel_C
  widgetCount=1
 }
 widgets [Settings_Panel {widgetClassPath=/FortniteCoreUI/Settings/SettingsPanel, bInherited=true, bIsVariable=true, widget=Settings_Panel}]
}
