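# Video options tab in the legacy Settings UI, a Widget Blueprint on FortTabVideoOptions_TempNew built around an inherited Settings_Panel.
# The panel's SetSafeZone named action asynchronously creates a SafeZoneEditor with input suspended, pushes it as a modal panel, and passes it the triggering setting.
# CenterOnTab centers on the panel, and a SettingValueChanged custom event broadcasts the tab's setting-changed delegate.
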WidgetBlueprint /FortniteCoreUI/Settings/Legacy/TabVideoOptions_New
graphs [EventGraph {nodes [
   n1 = "On Execute Named Action (Settings_Panel)"!() -> (OutputDelegate, then, Setting:Fort Setting Object Reference, Action:Name=None) {is=title K2Node_ComponentBoundEvent}
   > n2 = Switch on Name!(execute:Exec, Selection:Name=@n1.Action) -> (Default:Exec, SetSafeZone:Exec) {is=title K2Node_SwitchName, pinNames [SetSafeZone]}
   n3 = CreateWidgetAsync!(
    execute:Exec=@n2.SetSafeZone
    UserWidgetSoftClass:User Widget Soft Class Reference=/FortniteCoreUI/UI/Login/SafeZone/SafeZoneEditor.SafeZoneEditor_C
    OwningPlayer:Player Controller Object Reference=GetOwningPlayer()
    bSuspendInputUntilComplete:bool=true
   ) -> (then:Exec, OnComplete:Exec, OnFailure:Exec, UserWidget:User Widget Object Reference) {is=title K2Node_AsyncAction}
   TryPushWidgetAsModalPanel!(execute=@n3.OnComplete, PushWidget:User Widget Object Reference=@n3) -> (then)
   > n4 = Cast To SafeZoneEditor!(execute, Object=@n3) -> (then, CastFailed, AsSafe Zone Editor:Safe Zone Editor Object Reference) {is=title K2Node_DynamicCast}
   > SetChildSettingReference!(execute, self:Fort Safe Zone Editor Object Reference=@n4, SettingReference:Fort Setting Object Reference=@n1.Setting) -> (then)
   SettingValueChanged!() -> (OutputDelegate, then, Value:double=~) {is=title K2Node_CustomEvent}
   > Call On Tab Setting Changed!(execute:Exec, self:Fort Options Tab Object Reference) -> (then:Exec) {is=title K2Node_CallDelegate}
   Event CenterOnTab() {is=title K2Node_Event}
   > CenterWidget!(execute, ToCenter:Widget Object Reference=Get Settings_Panel() -> (Settings_Panel:Fort Settings Panel Object Reference)) -> (then)
  ]}]
variables []
widgets {
 info {
  inheritedWidgetCount=1
  parentClass=/Script/FortniteUI.FortTabVideoOptions_TempNew
  rootWidgetClass=/FortniteCoreUI/Settings/SettingsPanel.SettingsPanel_C
  widgetCount=1
 }
 widgets [Settings_Panel {widgetClassPath=/FortniteCoreUI/Settings/SettingsPanel, bInherited=true, bIsVariable=true, widget=Settings_Panel}]
}
